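# Sound cue for the poison dart trap's damage loop, layering a dart impact with a sustained looping sound.
# A looping wave fades in through an Enveloper volume curve and is mixed with one of four equally weighted, randomly chosen crossbow impact sounds.
# The mix passes through a Modulator at the root.
# It plays in the gadgets/items sound class at reduced volume, with its own attenuation override.
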SoundCue /FNGameplayCues/Sounds/Fort_Traps/PoisonDart_Trap/PoisonDart_DamageLoop_Cue
root=Modulator_0
nodes [
 {
  id=Enveloper_0
  children [WavePlayer_0]
  properties {pitchCurve {editorCurveData {keys [{time=0}]}}, volumeCurve {editorCurveData {keys [<at 0 0>,<at 1 1>]}}}
 }
 {id=Mixer_0, children [Enveloper_0,Random_1], properties {inputVolume [0.4,0.5]}}
 {id=Modulator_0, children [Mixer_0]}
 {id=Random_1, children [WavePlayer_6,WavePlayer_7,WavePlayer_8,WavePlayer_9], properties {weights [1,1,1,1]}}
 {
  id=WavePlayer_0
  properties {bLooping=true, soundWaveAssetPtr=&/PoisonDartTrap_DamageLoop_01}
 }
 {id=WavePlayer_6, properties {soundWaveAssetPtr=/FNGameplayCues/Athena/Sounds/Impacts/Crossbow_Impact_01}}
 {id=WavePlayer_7, properties {soundWaveAssetPtr=/FNGameplayCues/Athena/Sounds/Impacts/Crossbow_Impact_02}}
 {id=WavePlayer_8, properties {soundWaveAssetPtr=/FNGameplayCues/Athena/Sounds/Impacts/Crossbow_Impact_03}}
 {id=WavePlayer_9, properties {soundWaveAssetPtr=/FNGameplayCues/Athena/Sounds/Impacts/Crossbow_Impact_04}}
]
properties {
 bOverrideAttenuation=true
 maxDistance=1536
 soundClassObject=&Fort_Gadgets_Items
 volumeMultiplier=0.5
 attenuationOverrides {falloffDistance=1024, stereoSpread=256, attenuationShapeExtents {x=512}}
}
