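# Sound cue for the Poison Dart trap's damage tick.
# A Random node picks among four equally weighted damage tick waves, and a Modulator at the root varies the result.
# The cue plays at reduced volume in the Fort_Impacts sound class, with its own attenuation override for distance falloff.
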SoundCue /FNGameplayCues/Sounds/Fort_Traps/PoisonDart_Trap/PoisonDart_DamageTick_Cue
root=Modulator_0
nodes [
 {id=Modulator_0, children [Random_0]}
 {id=Random_0, children [WavePlayer_6,WavePlayer_3,WavePlayer_4,WavePlayer_5], properties {weights [1,1,1,1]}}
 {id=WavePlayer_3, properties {soundWaveAssetPtr=&/PoisonDartTrap_DamageTick_02}}
 {id=WavePlayer_4, properties {soundWaveAssetPtr=&/PoisonDartTrap_DamageTick_03}}
 {id=WavePlayer_5, properties {soundWaveAssetPtr=&/PoisonDartTrap_DamageTick_04}}
 {id=WavePlayer_6, properties {soundWaveAssetPtr=&/PoisonDartTrap_DamageTick_01}}
]
properties {
 bOverrideAttenuation=true
 maxDistance=1536
 soundClassObject=&Fort_Impacts
 volumeMultiplier=0.15
 attenuationOverrides {falloffDistance=1024, stereoSpread=256, attenuationShapeExtents {x=512}}
}
